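; BP_Discoverable_Football is the Blueprint for a football discoverable in the front-end environment.
; On BeginPlay it runs the parent logic, then stores the root component's world location in InitLoc, which is not read in this graph.
; The overlap, tick and special-interaction events are empty stubs, and the construction script only calls its parent.
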(event EventBeginPlay
  (|CallParentFunction)
  (Variables|Default|SetInitLoc (Transformation|GetWorldLocation (Variables|Transformation|GetRootComponent))))

(event Collision|EventActorBeginOverlap
  (params (OtherActor :type Actor%20Object%20Reference)))

(event EventTick
  (params (DeltaSeconds :type Float%20%28single-precision%29)))

(event EventCommit_OnSpecialInteraction)

(fn ConstructionScript
  (|CallParentFunction))
