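; Music player discoverable for the front-end environment that shows a particle effect during the saxophone interaction.
; When notified by the discoverable system, it checks the action tag against ERG.Interaction.Saxaphone, then activates its Niagara component on start and deactivates it otherwise.
; The camera tag is unused, and the BeginPlay, overlap and Tick events carry no active logic.
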(event EventBeginPlay
  (|CallParentFunction) ; This node is disabled and will not be called.\nDrag off pins to build functionality.
)

(event Collision|EventActorBeginOverlap
  (params (OtherActor :type Actor%20Object%20Reference)))

(event EventTick
  (params (DeltaSeconds :type Float%20%28single-precision%29)))

(event EventNotifiedByDiscoverable
  (params (ActionTag :type Gameplay%20Tag%20Structure) (CameraTag :type Gameplay%20Tag%20Structure) (bIsStart :type Boolean))
  (bind _niagara (Variables|Default|GetNiagara))
  (if (GameplayTags|Equal(GameplayTag) ActionTag "(TagName=\"ERG.Interaction.Saxaphone\")")
    (if bIsStart
      (Components|Activation|Activate _niagara false)
      (else
        (Components|Activation|Deactivate _niagara)))))

(fn ConstructionScript
  (|CallParentFunction))
