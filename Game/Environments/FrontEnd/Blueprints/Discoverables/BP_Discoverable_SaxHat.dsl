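; BP_Discoverable_SaxHat is a front-end environment discoverable Blueprint, apparently a hat tied to the saxophone interaction.
; On BeginPlay it runs parent logic and caches the DupeMesh world location as DisplayMeshInitLoc.
; When notified by a discoverable, it stores bIsStart and checks for the ERG.Interaction.Saxaphone tag, but the start branch performs no action.
; Overlap and Tick events are empty, and construction only calls the parent.
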(event EventBeginPlay
  (|CallParentFunction)
  (Variables|Default|SetDisplayMeshInitLoc (Transformation|GetWorldLocation (Variables|Default|GetDupeMesh))))

(event Collision|EventActorBeginOverlap
  (params (OtherActor :type Actor%20Object%20Reference)))

(event EventTick
  (params (DeltaSeconds :type Float%20%28single-precision%29)))

(event EventNotifiedByDiscoverable
  (params (ActionTag :type Gameplay%20Tag%20Structure) (CameraTag :type Gameplay%20Tag%20Structure) (bIsStart :type Boolean))
  (bind _output_get (Variables|Default|SetbIsStart bIsStart))
  (if (GameplayTags|Equal(GameplayTag) ActionTag "(TagName=\"ERG.Interaction.Saxaphone\")")
    (if _output_get
      _)))

(fn ConstructionScript
  (|CallParentFunction))
